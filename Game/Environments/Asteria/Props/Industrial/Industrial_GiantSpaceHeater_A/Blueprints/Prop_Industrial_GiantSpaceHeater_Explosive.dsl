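; Explosive variant of the industrial giant space heater prop, deciding which actors its explosion affects.
; DestroyHitActors gates each hit actor with a line-of-sight trace from an explosion offset before calling ExecuteExplosiononActor.
; Pawns require clear sight, while non-pawns are traced toward their collision bounds and also pass within a sure-hit distance, so clumped buildings still get destroyed.
; The construction script only calls the parent.
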(event EventDestroyHitActors
  (params (HitActor :type Actor%20Object%20Reference))
  (bind _output_get (Variables|Default|SetHit%20Actor HitActor))
  (Utilities|Casting|CastToFortPawn _output_get
    (:then
      ;[1 1 1] If we are dealing with pawns, we take into consideration Line of Sight
      (bind _in_line_of_sight (Class|MissionBlueprintFunctionLibrary|CheckBasicActorLineOfSight self (Variables|Default|GetExplosionTraceOffset) (Variables|Default|GetHit%20Actor) "0, 0, 0" false "TraceTypeQuery5" "Persistent" 15.0 true))
      (if _in_line_of_sight
        _))
    (:CastFailed
      ;[1 1 1] If we are dealing with non-pawns, we allow for some override on the LoS so we can destroy buildings that are clumped up
      (bind (_in_line_of_sight_2 _distance) (Class|MissionBlueprintFunctionLibrary|CheckBasicActorLineOfSight self (Variables|Default|GetExplosionTraceOffset) (Variables|Default|GetHit%20Actor) (- (Fortnite|GetActorCollisionComponentBounds (Variables|Default|GetHit%20Actor)) (Transformation|GetActorLocation (Variables|Default|GetHit%20Actor))) false "TraceTypeQuery5" "Persistent" 15.0 true))
      (if (or _in_line_of_sight_2 (<= _distance (Variables|Default|GetExplosionTraceSureHitDistance)))
        _)))
  (CallFunction|ExecuteExplosiononActor (Variables|Default|GetHit%20Actor)))

(fn ConstructionScript
  (|CallParentFunction))
