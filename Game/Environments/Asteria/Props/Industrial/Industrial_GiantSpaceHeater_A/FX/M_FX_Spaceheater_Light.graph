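# Additive, unlit, two-sided material for the giant space heater's light effect, pulsing over time with a phase offset from object world position.
# Emissive blends FalloffColor to CoreColor along UV V, scaled by the pulse and time-of-day brightness, then desaturated by time-of-day saturation and passed through shared FX emissive adjustments.
# Opacity multiplies V by a Z-axis fresnel with V-varying exponent, adding depth fade on all but Low quality.
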Material /Game/Environments/Asteria/Props/Industrial/Industrial_GiantSpaceHeater_A/FX/M_FX_Spaceheater_Light
graph {
 properties {
  bUseMaterialAttributes=true
  blendMode=BLEND_Additive
  pixelDepthOffsetMode=PDOM_Legacy
  shadingModel=MSM_Unlit
  twoSided=true
  usedShadingModels=Unlit
 }
 nodes [
  n1 = DotProduct(A=ObjectPositionWS(), B=Constant3Vector() {constant=<rgba 1 1 0 0>})
  n2 = Sine(None=Add(A=Divide(A=@n1) {constB=1024}, B=Time())) {period=4}
  n3 = Add(A=Add(A=Multiply(A=@n2) {constB=0.5}) {constB=0.5}) {constB=0.75}
  n4 = ComponentMask(None=TextureCoordinate()) {g=true}
  n5 = LinearInterpolate(
   A=VectorParameter() {parameterName=FalloffColor, defaultValue=<rgba 0 0 0 1>}.RGB
   B=VectorParameter() {parameterName=CoreColor, defaultValue=<rgba 0 0 0 1>}.RGB
   Alpha=@n4
  )
  n6 = ScalarParameter() {group=TOD, parameterName=Brightness_Morning}
  n7 = ScalarParameter() {group=TOD, parameterName=Brightness_Day}
  n8 = ScalarParameter() {group=TOD, parameterName=Brightness_Evening}
  n9 = ScalarParameter() {group=TOD, parameterName=Brightness_Night}
  n10 = ScalarParameter() {defaultValue=1, group=TOD, parameterName=Saturation_Morning}
  n11 = ScalarParameter() {defaultValue=1, group=TOD, parameterName=Saturation_Day}
  n12 = ScalarParameter() {defaultValue=1, group=TOD, parameterName=Saturation_Evening}
  n13 = ScalarParameter() {defaultValue=1, group=TOD, parameterName=Saturation_Night}
  n14 = Desaturation(
   None=Multiply(
    A=Multiply(A=@n3, B=@n5)
    B=MaterialFunctionCall(Value_Morning=@n6, Value_Day=@n7, Value_Evening=@n8, Value_Night=@n9) {materialFunction=&Scalar_Time_of_Day}.Result
   )
   Fraction=OneMinus(None=MaterialFunctionCall(Value_Morning=@n10, Value_Day=@n11, Value_Evening=@n12, Value_Night=@n13) {materialFunction=&Scalar_Time_of_Day}.Result)
  )
  n15 = ScalarParameter() {parameterName=FalloffEx}
  n16 = ScalarParameter() {parameterName=Falloff_Source}
  n17 = Multiply(
   A=@n4
   B=MaterialFunctionCall(Axis=Constant3Vector() {constant=<rgba 0 0 1 0>}, Exponent=LinearInterpolate(A=@n15, B=@n16, Alpha=@n4)) {materialFunction=&AxisAlignedFresnel}.Fresnel
  )
  n18 = MakeMaterialAttributes!(
   BaseColor
   Metallic
   Specular
   Roughness
   Anisotropy
   EmissiveColor=NamedRerouteDeclaration(None=MaterialFunctionCall(Emissive=@n14) {materialFunction=&MF_FX_EmissiveAdjustments}.Result) {name=Color, nodeColor=<rgba 0.875 0 1 1>}
   Opacity=DepthFade(Opacity=NamedRerouteDeclaration(None=@n17) {name=Opacity, nodeColor=<rgba 0 0.9062 1 1>})
   OpacityMask
   Normal
   Tangent
   WorldPositionOffset
   SubsurfaceColor
   ClearCoat
   ClearCoatRoughness
   AmbientOcclusion
   Refraction
   CustomizedUVs_0
   CustomizedUVs_1
   CustomizedUVs_2
   CustomizedUVs_3
   CustomizedUVs_4
   CustomizedUVs_5
   CustomizedUVs_6
   CustomizedUVs_7
   PixelDepthOffset
   ShadingModel
   Displacement
  )
  n19 = MakeMaterialAttributes(EmissiveColor=NamedRerouteUsage(), Opacity=NamedRerouteUsage())
  QualitySwitch(Default=@n18, Low=@n19, High=@n18, Medium=@n19, Epic=@n18)
 ]
}
